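# Post-copy AnimBP for the AgentSherbert Dew female body: adds secondary motion and toggles a material element for its matching pickaxe.
# The pose copied from the parent mesh gets right-bow bone offsets, spring Anim Dynamics on left and right bow and tongue bones, then a deformation Control Rig.
# Customization-load and weapon-change events check the Stage3 variant, then set "Hide Element 02" while that pickaxe is equipped outside skydiving; skydive events also update it.
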AnimBlueprint /BRCosmetics/Characters/Player/Female/Medium/Bodies/F_MED_AgentSherbert_Dew/Meshes/F_MED_AgentSherbert_Dew_AnimBP
target_skeleton=&/F_MED_AgentSherbert_Dew_Skeleton
graphs [
 AnimGraph {nodes [
   n1 = "Transform (Modify) Bone - Bone: dyn_R_bow_1_1"(
    ComponentPose=Local To Component(LocalPose=Copy Pose From Mesh() {bUseAttachedParent=true})
    Alpha=0.5
    Translation="-0.673,-0.122,1.989"
    Rotation="36.696,0.001,0.006"
    Scale="1,1,1"
   ) {
    rotationMode=BMM_Additive
    rotationSpace=BCS_BoneSpace
    translationMode=BMM_Replace
    translationSpace=BCS_BoneSpace
    alphaBoolBlend {blendOption=Linear}
    boneToModify {boneName=dyn_R_bow_1_1}
    rotation=<rot 36.696 0.001 0.006>
    translation=<-0.673 -0.122 1.989>
   }
   n2 = "Transform (Modify) Bone - Bone: dyn_R_bow_2_1"(ComponentPose=@n1, Alpha=1, Translation="-0.639,0.363,-0.403", Rotation="4.966,-2.131,-5.629", Scale="1,1,1") {
    rotationMode=BMM_Additive
    translationMode=BMM_Replace
    translationSpace=BCS_BoneSpace
    alphaBoolBlend {blendOption=Linear}
    boneToModify {boneName=dyn_R_bow_2_1}
    rotation=<rot 4.966 -2.131 -5.629>
    translation=<-0.639 0.363 -0.403>
   }
   n3 = Anim Dynamics - Bone: dyn_L_bow_1_1(ComponentPose=@n2, Alpha=1, ExternalForce="0,0,0") {
    angularDampingOverride=0.9999
    angularSpringConstant=50
    bAngularSpring=true
    bOverrideAngularDamping=true
    bUseGravityOverride=true
    simulationSpace=RootRelative
    alphaBoolBlend {blendOption=Linear}
    boundBone {boneName=dyn_L_bow_1_1}
    componentAppliedLinearAccClamp=<1000 1000 1000>
    componentLinearAccScale=<1 1 1>
    componentLinearVelScale=<1 1 1>
    gravityOverride {z=-920}
    physicsBodyDefinitions [{
      boundBone {boneName=dyn_L_bow_1_1}
      boxExtents=<2 2 2>
      constraintSetup {angularLimitsMax=<0 30 20>, angularLimitsMin=<0 -30 -3>, angularTarget=<0 2 0>}
      localJointOffset=<0.1 0 0>
     }]
    relativeSpaceBone {boneName=calf_twist_02_r}
   }
   n4 = Anim Dynamics - Bone: dyn_L_bow_2_1(ComponentPose=@n3, Alpha=1, ExternalForce="0,0,0") {
    angularDampingOverride=0.9999
    angularSpringConstant=50
    bAngularSpring=true
    bOverrideAngularDamping=true
    bUseGravityOverride=true
    simulationSpace=RootRelative
    alphaBoolBlend {blendOption=Linear}
    boundBone {boneName=dyn_L_bow_2_1}
    componentAppliedLinearAccClamp=<1000 1000 1000>
    componentLinearAccScale=<1 1 1>
    componentLinearVelScale=<1 1 1>
    gravityOverride {z=-920}
    physicsBodyDefinitions [{
      boundBone {boneName=dyn_L_bow_2_1}
      boxExtents=<2 2 2>
      constraintSetup {angularLimitsMax=<30 10 0>, angularLimitsMin=<-30 -40 0>, angularTarget=<0 0 1>}
      localJointOffset=<0.1 0 0>
     }]
    relativeSpaceBone {boneName=calf_twist_02_r}
   }
   n5 = Anim Dynamics - Bone: dyn_L_tongue_1(ComponentPose=@n4, Alpha=1, ExternalForce="0,0,0") {
    angularDampingOverride=1
    angularSpringConstant=100
    bAngularSpring=true
    bOverrideAngularDamping=true
    bUseGravityOverride=true
    simulationSpace=RootRelative
    alphaBoolBlend {blendOption=Linear}
    boundBone {boneName=dyn_L_tongue_1}
    componentAppliedLinearAccClamp=<1000 1000 1000>
    componentLinearAccScale=<0.3 0.3 0.3>
    componentLinearVelScale=<0.3 0.3 0.3>
    gravityOverride {z=-980}
    physicsBodyDefinitions [{
      boundBone {boneName=dyn_L_tongue_1}
      boxExtents=<5 5 5>
      constraintSetup {angularLimitsMax=<0 0 30>, angularLimitsMin=<0 0 -30>, angularTarget=<1 0 3>, linearAxesMax=<1 0 0>}
      localJointOffset=<1 1 1>
     }]
    relativeSpaceBone {boneName=calf_twist_02_r}
   }
   n6 = Anim Dynamics - Bone: dyn_R_bow_1_1(ComponentPose=@n5, Alpha=1, ExternalForce="0,0,0") {
    angularDampingOverride=0.9999
    angularSpringConstant=50
    bAngularSpring=true
    bOverrideAngularDamping=true
    bUseGravityOverride=true
    simulationSpace=RootRelative
    alphaBoolBlend {blendOption=Linear}
    boundBone {boneName=dyn_R_bow_1_1}
    componentAppliedLinearAccClamp=<1000 1000 1000>
    componentLinearAccScale=<1 1 1>
    componentLinearVelScale=<1 1 1>
    gravityOverride {z=-920}
    physicsBodyDefinitions [{
      boundBone {boneName=dyn_R_bow_1_1}
      boxExtents=<2 2 2>
      constraintSetup {angularLimitsMax=<5 5 10>, angularLimitsMin=<-20 -20 -20>, angularTarget=<0 0 -0.5>}
      localJointOffset=<0.1 0 0>
     }]
    relativeSpaceBone {boneName=calf_twist_02_r}
   }
   n7 = Anim Dynamics - Bone: dyn_R_bow_2_1(ComponentPose=@n6, Alpha=1, ExternalForce="0,0,0") {
    angularDampingOverride=0.9999
    angularSpringConstant=50
    bAngularSpring=true
    bOverrideAngularDamping=true
    bUseGravityOverride=true
    simulationSpace=RootRelative
    alphaBoolBlend {blendOption=Linear}
    boundBone {boneName=dyn_R_bow_2_1}
    componentAppliedLinearAccClamp=<1000 1000 1000>
    componentLinearAccScale=<1 1 1>
    componentLinearVelScale=<1 1 1>
    gravityOverride {z=-920}
    physicsBodyDefinitions [{
      boundBone {boneName=dyn_R_bow_2_1}
      boxExtents=<2 2 2>
      constraintSetup {angularLimitsMax=<30 5 30>, angularLimitsMin=<-5 -20 -5>, angularTarget=<0 -1 0>}
      localJointOffset=<0.1 0 0>
     }]
    relativeSpaceBone {boneName=calf_twist_02_r}
   }
   n8 = Anim Dynamics - Bone: dyn_R_tongue_1(ComponentPose=@n7, Alpha=1, ExternalForce="0,0,0") {
    angularDampingOverride=1
    angularSpringConstant=100
    bAngularSpring=true
    bOverrideAngularDamping=true
    bUseGravityOverride=true
    simulationSpace=RootRelative
    alphaBoolBlend {blendOption=Linear}
    boundBone {boneName=dyn_R_tongue_1}
    componentAppliedLinearAccClamp=<1000 1000 1000>
    componentLinearAccScale=<0.3 0.3 0.3>
    componentLinearVelScale=<0.3 0.3 0.3>
    gravityOverride {z=-980}
    physicsBodyDefinitions [{
      boundBone {boneName=dyn_R_tongue_1}
      boxExtents=<5 5 5>
      constraintSetup {angularLimitsMax=<0 0 30>, angularLimitsMin=<0 0 -30>, angularTarget=<1 0 3>, linearAxesMax=<1 0 0>}
      localJointOffset=<1 1 1>
     }]
    relativeSpaceBone {boneName=calf_twist_02_r}
   }
   n9 = Control Rig(Source=Component To Local(ComponentPose=@n8)) {
    alphaBoolBlend {blendOption=Linear}
    controlRigAssetReference {blueprintRigClass=&/F_MED_AgentSherbert_Dew_Deforms_ControlRig.F_MED_AgentSherbert_Dew_Deforms_ControlRig_C}
   }
   Output Pose(Result=@n9) {layerGroup=None}
  ]}
 EventGraph {nodes [
   n1 = Event OnCharacterCustomizationLoaded()
   n2 = Get Dynamic Material Instance(execute=[@n1.then, @n8.then], SkeletalMeshComponent=GetPartSkeletalMeshComponent())
   > Branch(Condition=CharacterItemDefinitionHasAnyMetaTag(Tags=Make Array("[0]"=(TagName="Cosmetics.Variant.Property.Stage3"))))
   n3 = PickaxeItemDefinitionHasAnyMetaTag(Tags=Make Array("[0]"=(TagName="Cosmetics.MetaTags.AgentSherbert")), bRequiresEquipped=true)
   > n4 = Branch(Condition=@n3)
   > n5 = Branch(Condition=Get bIsSkyDiving())
   SetScalarParameterValue(execute=[@n6.then, @n5.then, @n7.else, @n4.else], self=@n2, ParameterName=Hide Element 02, Value=0)
   SetScalarParameterValue(execute=[@n5.else, @n7.then], self=@n2, ParameterName=Hide Element 02, Value=1)
   n6 = Event OnBeginSkydiving()
   Event OnEndSkydiving()
   > n7 = Branch(Condition=@n3)
   n8 = Event OnWeaponChanged()
   Event BlueprintUpdateAnimation()
  ]}
]
variables [Material instanse {type=MaterialInstanceDynamic}]
